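# Niagara system for a toon-lightning effect on a character's skeletal mesh, with ribbon arcs and a mesh overlay.
# Bone_Flow_hands repeatedly bursts short-lived, 15-point ribbon groups that FN_BoneFlow shapes along two alternating bone IDs with random offsets.
# Two ribbon renderers draw the arcs, which fade over life and dim by time of day.
# Empty spawns one fade-in, mesh-rendered particle on the skeleton, only for the local player or front-end preview.
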NiagaraSystem /BRCosmetics/Effects/Fort_Effects/GameplayPulginsAssets/ToonLightning/LightningRibbon/NS_WholebodyFX
dependencies {
 usedDataInterfaces [&NiagaraDataInterfaceCurve]
 usedDynamicInputs [&InvertBool]
 usedModules [
  &self:SystemSpawnScript.SystemScriptSource.SystemScriptGraph.NiagaraNodeAssignment_0.SetVariables_#1
  &self:SystemSpawnScript.SystemScriptSource.SystemScriptGraph.NiagaraNodeAssignment_1.SetVariables_#2
  &SystemState
 ]
}
emitters {
 Bone_Flow_hands {
  emitterUpdateScript [
   &EmitterState(Delay First Loop Only=true, Life Cycle Mode=::NewEnumerator1, Loop Duration=0.05, UseLoopDelay=true)
   &SpawnBurst_Instantaneous(
    Spawn Count=&MakeIntFromFloat(Float=&FloatFromCurve(
      FloatCurve=DataInterface() {propertyValues {Curve {keys [
          {interpMode=RCIM_Cubic, tangentMode=RCTM_User, time=0, value=0}
          <at 0.1 1>
          {interpMode=RCIM_Cubic, tangentMode=RCTM_User, time=0.5, value=1}
          <at 0.6 1>
          {tangentMode=RCTM_User, time=1, value=0}
         ]}}}
      CurveIndex=$Emitter.NormalizedLoopAge
      Scale Curve=15
     ))
    Use Spawn Probability=false
    UseLoopCountLimit=false
   )
  ]
  particleSpawnScript [
   &FN_SetSpawnGroup(Inclusivity=true, Normalize Spawn Group=true, Particles Per Group=15, Pattern=::NewEnumerator0, Set Ribbon ID and Ribbon Link Order=true)
   &InitializeParticle(
    Lifetime Max=0.15
    Lifetime Min=0.13
    Lifetime Mode=::NewEnumerator1
    Ribbon Width=6
    Ribbon Width Mode=::NewEnumerator1
    UsePositionOffset=false
    Write Material Random=true
   )
   &CameraOffset(Camera Offset Amount=10)
  ]
  particleUpdateScript [
   &ParticleState()
   &FN_BoneFlow(
    Angle:NiagaraFloat=&Multiply_Float(A=$Particles.Normalized Spawn Group, B=2.0626)
    BoneID:NiagaraInt32=&ModuloInt(A=$Particles.SpawnGroup, B=2)
    Enable Position Offsets=true
    Height:NiagaraFloat=&Multiply_Float(A=$Particles.Normalized Spawn Group, B=&RandomRangeFloat(Minimum=1, Maximum=1.1))
    Offset Mid:Vector3f=&RandomRangeVector(Minimum=<-10 -10 0>, Maximum=<10 10 0>)
    Radius:NiagaraFloat=&Multiply_Float(
     A=&FloatFromCurve(
      FloatCurve=DataInterface() {propertyValues {Curve {keys [
          <at 0 3 RCIM_Cubic>
          {interpMode=RCIM_Cubic, tangentMode=RCTM_User, time=0.0945, value=9.5896}
          {arriveTangent=-0.0483, interpMode=RCIM_Cubic, leaveTangent=-0.0483, time=0.4976, value=10.0326}
          {interpMode=RCIM_Cubic, tangentMode=RCTM_User, time=0.898, value=9.5507}
          <at 1 3>
         ]}}}
      CurveIndex=$Particles.Normalized Spawn Group
     )
     B=1
    )
    SkeletalMesh:NiagaraDataInterfaceSkeletalMesh=$User.SK_Mesh
   )
   &DynamicMaterialParameters(
    Index 0 Param 1=1
    Index 0 Param 4=1
    Module.Param0WriteEnabled=true
    Module.Param1WriteEnabled=true
    Module.Param2WriteEnabled=true
    Module.Param3WriteEnabled=true
   ) {enabled=false}
   &DynamicMaterialParameters(
    Index 0 Param 4=1
    Index 1 Param 2=&FloatFromCurve(FloatCurve=DataInterface() {propertyValues {Curve {keys [<at 0 0>,<at 0.25 1>]}}})
    Index 1 Param 3=-0.1
    Module.Param0WriteEnabled=true
    Module.Param0WriteEnabled1=true
    Module.Param0WriteEnabled2=true
    Module.Param1WriteEnabled=true
    Module.Param1WriteEnabled1=true
    Module.Param1WriteEnabled2=true
    Module.Param2WriteEnabled=true
    Module.Param2WriteEnabled1=true
    Module.Param2WriteEnabled2=true
    Module.Param3WriteEnabled=true
    Module.Param3WriteEnabled1=true
    Module.Param3WriteEnabled2=true
    Write Parameter Index 1=true
    Write Parameter Index 2=true
   ) {module=DynamicMaterialParameters001}
   &ScaleRibbonWidth(Ribbon Width Scale=&FloatFromCurve(
     FloatCurve=DataInterface() {propertyValues {Curve {keys [<at 0.25 1>,<at 0.3 0.6>,<at 0.7 0.6>,<at 0.75 1>]}}}
     CurveIndex=$Particles.RibbonLinkOrder
    )) {enabled=false}
   /SetVariables(Particles.Glow Color=&MakeLinearColorFromVectorAndFloat(
     "Vector (RGB)"=&MakeVectorFromLinearColorRGB(Color=&MultiplyLinearColor(Linear Color=<rgba 1 1 1 1>, Scale Factor=&MakeVector4_FromFloat(Float=1)))
     "Float (Alpha)"=&Multiply_Float(
      A=&FloatFromCurve(FloatCurve=DataInterface() {propertyValues {Curve {keys [
           {interpMode=RCIM_Cubic, tangentMode=RCTM_User, time=0, value=1}
           {interpMode=RCIM_Cubic, leaveTangent=-7.6143, tangentMode=RCTM_Break, time=0.5, value=1}
           {arriveTangent=-0.1251, interpMode=RCIM_Cubic, leaveTangent=-0.1251, tangentMode=RCTM_User, time=0.9, value=0}
          ]}}})
      B=3
     )
    )) {isSetParameters=true}
   &ScaleColor(
    Scale Alpha=&Multiply_Float(
     A=1
     B=&FloatFromCurve(FloatCurve=DataInterface() {propertyValues {Curve {keys [
          {tangentMode=RCTM_User, time=0, value=1}
          {arriveTangent=-6.3702, interpMode=RCIM_Cubic, leaveTangent=-6.3702, tangentMode=RCTM_User, time=0.5, value=1}
          {arriveTangent=-0.5139, interpMode=RCIM_Cubic, leaveTangent=-0.5139, tangentMode=RCTM_User, time=1, value=0}
         ]}}})
    )
    Scale RGB=&Multiply_VectorByFloat(
     Vector=<1 1 1>
     Float=&Multiply_Float(
      A=&FloatFromCurve(
       FloatCurve=DataInterface() {propertyValues {Curve {keys [
           <at 0 2>
           <at 0.3 2>
           {arriveTangent=-2, interpMode=RCIM_Cubic, leaveTangent=-2, time=0.5, value=1}
           {arriveTangent=2, interpMode=RCIM_Cubic, leaveTangent=2, time=0.8, value=1}
           <at 1 2>
          ]}}}
       Scale Curve=0.5
      )
      B=&Multiply_Float(A=0.9, B=1)
     )
    )
    ScaleA=true
    ScaleRGB=true
   )
   /SetVariables(Particles.RibbonW2:NiagaraFloat=&Multiply_Float(A=$Particles.RibbonWidth, B=1.5)) {isSetParameters=true}
   &ScaleColor(
    Scale RGB=&VectorFromFloat(Value=&FNTimeOfDayScalar(Night Scale=0.7, Morning Scale=1, Evening Scale=1))
    ScaleA=true
    ScaleRGB=true
   ) {module=ScaleColor001}
  ]
  rendererClasses [&NiagaraRibbonRendererProperties]
  renderers [
   {rendererClass=&NiagaraRibbonRendererProperties}
   {rendererClass=&NiagaraRibbonRendererProperties, rendererIndex=1}
  ]
 }
 Empty {
  emitterUpdateScript [
   &EmitterState(
    Inactive Response=::NewEnumerator1
    Life Cycle Mode=::NewEnumerator1
    Loop Behavior=::NewEnumerator1
    Loop Duration Mode=::NewEnumerator1
    UseLoopDelay=false
   )
   &SpawnBurst_Instantaneous(
    Spawn Count=&MakeCustomIntFromBool(
     Bool=$User.bIsFrontEndPreview
     True INT=1
     False INT=&MakeCustomIntFromBool(Bool=$User.bIsLocalPlayer, True INT=1, False INT=0)
    )
    Use Spawn Probability=false
    UseLoopCountLimit=false
   )
  ]
  particleSpawnScript [&InitializeParticle(
    Color Mode=::NewEnumerator0
    Mesh Scale=<1.8 0.8 0.8>
    Mesh Scale Mode=::NewEnumerator4
    Sprite Size=<118.0405 198.6727>
    Sprite Size Mode=::NewEnumerator4
    UsePositionOffset=false
    Write Material Random=true
   )]
  particleUpdateScript [
   &ParticleState(Kill Particles When Lifetime Has Elapsed=false, Let Infinitely Lived Particles Die When Emitter Deactivates=true)
   &SkeletalMeshLocation(Bone Sampling Mode=::NewEnumerator4, MaskSpawnGroup=false, Skeletal Mesh=$System.TargetSM)
   &CameraOffset(Camera Offset Amount=38.6542)
   &ScaleColor(
    Scale Alpha=&FloatFromCurve(FloatCurve=DataInterface() {propertyValues {Curve {keys [<at 0 0>,<at 0.4 1>]}}}, CurveIndex=$Emitter.Age, Scale Curve=2)
    ScaleA=true
    ScaleRGB=true
   )
  ]
  rendererClasses [&NiagaraMeshRendererProperties]
  renderers [
   {rendererClass=&NiagaraMeshRendererProperties, rendererIndex=0}
   {rendererClass=&NiagaraMeshRendererProperties, rendererIndex=1}
  ]
 }
}
summary {
 systemName=NS_WholebodyFX
 emitters [
  {emitterName=Empty, rendererClasses [&NiagaraMeshRendererProperties]}
  {emitterName=Bone_Flow_hands, rendererClasses [&NiagaraRibbonRendererProperties]}
 ]
 userVariables [
  User.bIsFrontEndPreview {type=NiagaraBool}
  User.bIsLocalPlayer {type=NiagaraBool, defaultValue=true}
  User.SK_Mesh {
   type=NiagaraDataInterfaceSkeletalMesh
   defaultValue {
    struct=NiagaraExt_VariableValue_DataInterface
    value {dataInterface=&self:NiagaraDataInterfaceSkeletalMesh_1}
   }
  }
 ]
}
